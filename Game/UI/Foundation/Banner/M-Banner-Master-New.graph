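# Translucent UI master material for banners, compositing a shape, patterned gradient, tinted icon and optional decorative border.
# A texture or rounded-corner SDF masks the patterned two-color gradient, with the icon scalable through an animation parameter.
# ShowBorder overlays a packed-mask frame shaded by three colors plus a shine swipe, and BorderOnly can drop the interior opacity.
# The Diamond switch adds time-driven hue shifting and blendable thin-film iridescence to the border.
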Material /Game/UI/Foundation/Banner/M-Banner-Master-New
graph {
 properties {
  bUsedWithStaticMesh=false
  blendMode=BLEND_Translucent
  materialDomain=MD_UI
  pixelDepthOffsetMode=PDOM_Legacy
  usedShadingModels=Default Lit
 }
 nodes [
  n1 = Constant2Vector() {g=1, r=1.28}
  n2 = ScalarParameter() {desc="(0 - 1) Animate scale transition 'ScaleFrom' - 'ScaleTo'", group=Item, parameterName=ScaleAnimation, sliderMax=1, sortPriority=4}
  n3 = LinearInterpolate(
   A=ScalarParameter() {defaultValue=1, desc=Start value for scale animation., group=Item, parameterName=ScaleFrom, sortPriority=2}
   B=ScalarParameter() {defaultValue=1.1, desc=End value for scale animation., group=Item, parameterName=ScaleTo, sortPriority=3}
   Alpha=@n2
  )
  Multiply(A=@n3, B=@n1)
  n4 = Constant2Vector() {g=0.55, r=0.6}
  Multiply(A=@n4, B=@n3)
  n5 = MaterialFunctionCall(
   Corner Radius=MaterialFunctionCall(
    R=ScalarParameter() {defaultValue=20, group=Layout, parameterName=TopLeftCornerRadius, sortPriority=2}
    G=ScalarParameter() {defaultValue=20, group=Layout, parameterName=BottomLeftCornerRadius, sortPriority=3}
    B=ScalarParameter() {defaultValue=20, group=Layout, parameterName=TopRightCornerRadius, sortPriority=4}
    A=ScalarParameter() {defaultValue=20, group=Layout, parameterName=BottomRightCornerRadius, sortPriority=5}
   ) {materialFunction=&AppendMany}.RGBA
   Radius in Pixels=ScalarParameter() {defaultValue=1, group=Layout, parameterName=UseValueInPixels, sortPriority=1}
  ) {materialFunction=&MF_UI_CoreUI_RoundedMask}
  n6 = Constant() {r=0.5}
  n7 = MaterialFunctionCall(DistanceField=Add(A=@n5.SDF) {constB=0.5}, Threshold=@n6) {materialFunction=&MF_UI_AntiAliasing}
  n8 = Add(A=@n5.Outline SDF) {constB=0.5}
  NamedRerouteDeclaration(None=Subtract(
    A=NamedRerouteDeclaration(None=@n7.Inner) {name=Fill, nodeColor=<rgba 0 0 0 1>}
    B=MaterialFunctionCall(DistanceField=@n8, Threshold=@n6) {materialFunction=&MF_UI_AntiAliasing}.Outer
   )) {name=Outline, nodeColor=<rgba 0 0 0 1>}
  n9 = Rotator(Time=Constant() {r=6.8}) {speed=1}
  n10 = TextureSample(UVs=Multiply(A=Panner(Coordinate=@n9) {speedX=0.5})) {texture=/Game/Global/Textures/Default/Thinfilm/T_ThinFilm_Spectrum_COLOR}
  n11 = Rotator(Time=Constant() {r=3.4}) {speed=1}
  n12 = TextureSample(UVs=Multiply(A=Panner(Coordinate=@n11) {speedX=0.5})) {texture=&T_ThinFilm_Spectrum_COLOR}
  n13 = Min(A=AppendVector(A=AppendVector(A=@n10.G, B=@n10.G), B=@n10.R), B=AppendVector(A=AppendVector(A=@n12.R, B=@n12.R), B=@n10.B))
  n14 = Desaturation(None=Clamp(None=Add(A=@n13) {constB=0.2}), Fraction=ScalarParameter() {defaultValue=0.8, parameterName=Desaturation})
  n15 = TextureSampleParameter2D() {parameterName=Texture, samplerType=SAMPLERTYPE_Masks, texture=/Game/UI/Foundation/Banner/BR-Banner/Textures/Season2/T-T-Banner-S2-01}
  n16 = MaterialFunctionCall(Alpha=@n15.G, Min=ScalarParameter() {parameterName=Dark SS Min}, Max=ScalarParameter() {defaultValue=0.5, parameterName=Dark SS Max}) {materialFunction=&SmoothStep}
  n17 = MaterialFunctionCall(
   Alpha=@n15.G
   Min=ScalarParameter() {defaultValue=0.25, parameterName=Mid SS Min}
   Max=ScalarParameter() {defaultValue=1, parameterName=Mid SS Max}
  ) {materialFunction=&SmoothStep}
  n18 = Power(Base=@n15.G) {constExponent=0.6}
  n19 = Add(A=@n18, B=Add(A=@n16.Result, B=@n17.Result))
  n20 = MaterialFunctionCall!(
   Sine Phase=Multiply(A=Time(), B=ScalarParameter() {defaultValue=1, parameterName=Hue Shift Time})
   Value 1=ScalarParameter() {defaultValue=-6, parameterName=Hue Shift Percent 1}
   Value 2=ScalarParameter() {defaultValue=6, parameterName=Hue Shift Percent 2}
  ) {materialFunction=/Engine/Functions/Engine_MaterialFunctions02/Utility/Sine_Remapped}
  n21 = VectorParameter() {group=Color, parameterName=DarkShade, defaultValue=<rgba 0.644 0.2246 0.0303 1>}
  n22 = LinearInterpolate(
   B=StaticSwitchParameter(
    True=MaterialFunctionCall(Hue Shift Percentage=@n20.Remapped Sine, Texture=@n21.RGB) {materialFunction=&HueShift}.Result
    False=@n21.RGB
   ) {parameterName=Diamond}
   Alpha=@n18
  )
  n23 = VectorParameter() {group=Color, parameterName=MidShade, defaultValue=<rgba 0.6038 0.3564 0.063 1>}
  n24 = StaticSwitchParameter(
   True=MaterialFunctionCall(Hue Shift Percentage=@n20.Remapped Sine, Texture=@n23.RGB) {materialFunction=&HueShift}.Result
   False=@n23.RGB
  ) {parameterName=Diamond}
  n25 = VectorParameter() {group=Color, parameterName=LightShade, defaultValue=<rgba 1 0.87 0.405 1>}
  n26 = MaterialFunctionCall(Hue Shift Percentage=@n20.Remapped Sine, Texture=@n25.RGB) {materialFunction=&HueShift}
  n27 = MaterialFunctionCall!(In_Uv) {materialFunction=/Game/UI/Foundation/Materials/M-Function-ShineSwipe}
  n28 = Add(
   A=LinearInterpolate(
    A=LinearInterpolate(A=@n22, B=@n24, Alpha=@n16.Result)
    B=StaticSwitchParameter(True=@n26.Result, False=@n25.RGB) {parameterName=Diamond}
    Alpha=@n17.Result
   )
   B=Multiply(A=LinearInterpolate(B=@n27.Result, Alpha=@n15.B), B=ScalarParameter() {defaultValue=1, group=Shine, parameterName=Shine-Alpha})
  )
  n29 = Add(A=Multiply(A=Multiply(A=@n14, B=@n19), B=ScalarParameter() {parameterName=DiamondBlend}), B=@n28)
  n30 = Add(A=TextureSampleParameter2D() {parameterName=BannerPattern, texture=/Game/UI/HomeBase/BannerWidget/HB-Banner-Pattern-2}.A) {constB=0.6}
  n31 = ScalarParameter() {defaultValue=1, parameterName=EnableBannerMask}
  n32 = TextureCoordinate() {coordinateIndex=1}
  n33 = Constant2Vector() {g=1, r=1}
  n34 = ScalarParameter() {parameterName=OffsetX, sliderMax=1}
  n35 = ScalarParameter() {defaultValue=-0.06, parameterName=OffsetY, sliderMax=1, sliderMin=-1}
  n36 = MaterialFunctionCall(UVs=Add(A=Multiply(A=@n32, B=@n33), B=AppendVector(A=@n34, B=@n35)), Texture Scale=Constant2Vector() {g=1, r=1}) {materialFunction=&ScaleUVsByCenter}
  n37 = Add() {constB=0}
  n38 = StaticSwitchParameter(
   True=Add(
    A=MaterialFunctionCall(Texture Scale=@n4) {materialFunction=&ScaleUVsByCenter}.UVs
    B=Constant2Vector() {g=0.01}
   )
   False=StaticSwitchParameter(
    True=Add(A=@n36.UVs, B=@n37)
    False=MaterialFunctionCall(Texture Scale=@n1) {materialFunction=&ScaleUVsByCenter}.UVs
   ) {group=Square Rounded Corners, parameterName=SquareRoundedCorners, sortPriority=1}
  ) {defaultValue=true, parameterName=ShowBorder}
  n39 = TextureSampleParameter2D(UVs=@n38) {parameterName=BannerShape, texture=/Game/UI/Foundation/Textures/Banner/Shapes/banner_shape_standard}
  n40 = StaticSwitchParameter(True=NamedRerouteUsage(), False=@n39.RGB) {group=Square Rounded Corners, parameterName=SquareRoundedCorners, sortPriority=1}
  n41 = TextureCoordinate()
  n42 = Constant2Vector() {g=-0.575, r=-0.5}
  n43 = Subtract(A=Abs(None=Add(A=@n41, B=@n42) {constB=-0.5})) {constB=0.375}
  n44 = Clamp(None=MaterialFunctionCall(Vector 2=Ceil(None=@n43)) {materialFunction=/Engine/Functions/Engine_MaterialFunctions02/Utility/VectorLength}.V2 Length)
  n45 = StaticSwitchParameter(True=LinearInterpolate(A=@n39.RGB, Alpha=@n44) {constB=0}, False=@n40) {defaultValue=true, parameterName=ShowBorder}
  n46 = LinearInterpolate(
   A=Constant3Vector() {constant=<rgba 1 1 1 1>}
   B=@n45
   Alpha=StaticSwitchParameter(True=@n40, False=@n31) {group=Square Rounded Corners, parameterName=SquareRoundedCorners, sortPriority=1}
  )
  n47 = VectorParameter() {parameterName=RGBA0, defaultValue=<rgba 0.2279 0.3876 0.7991 1>}
  n48 = VectorParameter() {parameterName=RGBA1, defaultValue=<rgba 0.1339 0.286 0.525 1>}
  n49 = Rotator(
   Coordinate=MaterialFunctionCall() {materialFunction=&GetUserInterfaceUV}.Normalized UV
   Time=ScalarParameter() {group=Linear Gradient Settings, parameterName=ColorRotation}
  )
  n50 = MaterialFunctionCall(
   Alpha=Subtract(A=ComponentMask(None=@n49) {g=true}, B=ScalarParameter() {defaultValue=-0.1, parameterName=ColorOffset})
   Min=ScalarParameter() {group=Linear Gradient Settings, parameterName=RGBAMin}
   Max=ScalarParameter() {defaultValue=1, group=Linear Gradient Settings, parameterName=RGBAMax}
  ) {materialFunction=&SmoothStep}
  n51 = ScalarParameter() {defaultValue=1, desc=Start value for scale animation., group=Item, parameterName=ScaleFrom, sortPriority=2}
  n52 = ScalarParameter() {defaultValue=1.1, desc=End value for scale animation., group=Item, parameterName=ScaleTo, sortPriority=3}
  n53 = ScalarParameter() {desc="(0 - 1) Animate scale transition 'ScaleFrom' - 'ScaleTo'", group=Item, parameterName=ScaleAnimation, sliderMax=1, sortPriority=4}
  n54 = TextureSampleParameter2D(UVs=MaterialFunctionCall(UVs=@n38, Texture Scale=LinearInterpolate(A=@n51, B=@n52, Alpha=@n53)) {materialFunction=&ScaleUVsByCenter}.UVs) {parameterName=BannerIcon, texture=/Game/UI/Foundation/Textures/Banner/Standard/T-Banners-Icons-Waves-L}
  n55 = Multiply(A=StaticSwitchParameter(True=LinearInterpolate(A=@n54.R, Alpha=@n44) {constB=0}, False=@n54.R) {defaultValue=true, parameterName=ShowBorder}) {constB=2}
  n56 = LinearInterpolate(
   A=Multiply(A=Multiply(A=@n30, B=@n46), B=LinearInterpolate(A=@n47.RGB, B=@n48.RGB, Alpha=@n50.Result))
   B=Multiply(B=VectorParameter() {parameterName=BannerIconColor, defaultValue=<rgba 1 1 1 0>}.RGB) {constA=1}
   Alpha=Multiply(
    A=Clamp(None=@n55)
    B=MaterialFunctionCall(Float3=@n46) {materialFunction=&BreakOutFloat3Components}.R
   )
  )
  n57 = Multiply(A=@n15.R, B=ScalarParameter() {defaultValue=1, parameterName=Alpha})
  StaticSwitchParameter(True=LinearInterpolate(A=@n56, B=StaticSwitchParameter(True=@n29, False=@n28) {parameterName=Diamond}, Alpha=@n57), False=@n56) {defaultValue=true, parameterName=ShowBorder}
  n58 = OneMinus(None=ScalarParameter() {parameterName=BorderOnly})
  n59 = LinearInterpolate(B=ComponentMask(None=@n45) {r=true}, Alpha=@n31) {constA=1}
  StaticSwitchParameter(True=Add(A=@n57, B=Multiply(A=@n58, B=@n59)), False=@n59) {defaultValue=true, parameterName=ShowBorder}
 ]
}
